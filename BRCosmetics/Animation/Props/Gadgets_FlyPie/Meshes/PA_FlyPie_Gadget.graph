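# Physics asset for the FlyPie gadget prop in BRCosmetics, defining its collision bodies and joint constraints.
# Each of the root, main and pie_JNT bones has one capsule: a very wide, short one on root, a tiny one on main, and an offset, rotated one on pie_JNT.
# Two constraints chain pie_JNT to main and main to root, with matching swing and twist limit angles.
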PhysicsAsset /BRCosmetics/Animation/Props/Gadgets_FlyPie/Meshes/PA_FlyPie_Gadget
bodies {
 main [{shapeType=Capsule, length=0.505, radius=0.505, rotation=<rot 0 -90 0>}]
 pie_JNT [{shapeType=Capsule, length=21.9109, radius=32.3665, center=<-0.0734 0.2506 8.2566>, rotation=<rot 0 -118.3005 -89.5419>}]
 root [{shapeType=Capsule, length=0.505, radius=56.0134, rotation=<rot 0 -90 0>}]
}
constraints [
 {bone1Name=main, bone2Name=root, swing1LimitDegrees=45, swing2LimitDegrees=45, twistLimitDegrees=45}
 {bone1Name=pie_JNT, bone2Name=main, swing1LimitDegrees=45, swing2LimitDegrees=45, twistLimitDegrees=45}
]
